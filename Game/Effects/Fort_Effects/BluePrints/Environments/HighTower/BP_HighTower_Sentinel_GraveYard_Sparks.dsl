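; Blueprint actor for the spark effects in the HighTower Sentinel Graveyard environment, adding sound to the particle bursts.
; It handles particle spawn events from three spark emitters (Sparks1, Sparks2, Sparks3).
; Each event plays the shared Big Sparks Sound at the spawn event's location, so audio follows the particles.
; BeginPlay, Tick, overlap and the construction script are empty.
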(event EventBeginPlay)

(event OnParticleSpawn(Sparks1)
  (params (EventName :type Name :default "None") (EmitterTime :type Float%20%28single-precision%29) (Location :type Vector :default "0, 0, 0") (Velocity :type Vector :default "0, 0, 0"))
  ;[1 1 1] Spawn sound from here, location should be emitter location if needed
  (Audio|PlaySoundatLocation (Variables|Default|GetBig%20Sparks%20Sound) Location))

(event OnParticleSpawn(Sparks2)
  (params (EventName :type Name :default "None") (EmitterTime :type Float%20%28single-precision%29) (Location :type Vector :default "0, 0, 0") (Velocity :type Vector :default "0, 0, 0"))
  ;[1 1 1] Spawn sound from here, location should be emitter location if needed
  (Audio|PlaySoundatLocation (Variables|Default|GetBig%20Sparks%20Sound) Location))

(event OnParticleSpawn(Sparks3)
  (params (EventName :type Name :default "None") (EmitterTime :type Float%20%28single-precision%29) (Location :type Vector :default "0, 0, 0") (Velocity :type Vector :default "0, 0, 0"))
  ;[1 1 1] Spawn sound from here, location should be emitter location if needed
  (Audio|PlaySoundatLocation (Variables|Default|GetBig%20Sparks%20Sound) Location))

(event Collision|EventActorBeginOverlap
  (params (OtherActor :type Actor%20Object%20Reference)))

(event EventTick
  (params (DeltaSeconds :type Float%20%28single-precision%29)))

(fn ConstructionScript)
